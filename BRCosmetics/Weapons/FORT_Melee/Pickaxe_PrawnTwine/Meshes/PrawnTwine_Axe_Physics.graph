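# Physics asset for the PrawnTwine pickaxe mesh, covering the bones of its mailbox door.
# It has two capsule bodies: a tiny one on the base door bone and a larger, offset one on the dynamic door bone.
# A single constraint joins the dynamic bone to the base, locking twist and swing1 so that swing2 is the only unlocked angular axis, acting as a hinge.
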PhysicsAsset /BRCosmetics/Weapons/FORT_Melee/Pickaxe_PrawnTwine/Meshes/PrawnTwine_Axe_Physics
bodies {
 base_mailboxdoor [{shapeType=Capsule, length=0.505, radius=0.505, shapeName=base_mailboxdoor_capsule, rotation=<rot 0 -90 0>}]
 dyn_mailboxdoor [{shapeType=Capsule, length=11, radius=9, shapeName=dyn_mailboxdoor_capsule, center=<14 0 0>, rotation=<rot 90 0 0>}]
}
constraints [{
  bone1Name=dyn_mailboxdoor
  bone2Name=base_mailboxdoor
  swing1LimitDegrees=1
  swing1Motion=Locked
  swing2LimitDegrees=30
  twistLimitDegrees=1
  twistMotion=Locked
 }]
